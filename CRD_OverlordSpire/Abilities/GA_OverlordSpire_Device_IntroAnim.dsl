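; Gameplay ability for the Overlord Spire device's intro animation.
; On activation it casts the owning actor to Device_OverlordSpire and calls SetSpireHidden with false to reveal the spire.
; It then hands off to the parent ability's activation logic.
; The end-ability override is only a disabled placeholder.
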(event Ability|EventActivateAbility
  (bind _asdevice_overlord_spire (Utilities|Casting|CastToDevice_OverlordSpire (Ability|GetOwningActorfromActorInfo))
    (:then
      (StateFlow|SetSpireHidden _asdevice_overlord_spire false))
    (:CastFailed
      _))
  (|CallParentFunction))

(event Ability|EventOnEndAbility
  (params (bWasCancelled :type Boolean))
  (|CallParentFunction bWasCancelled) ; This node is disabled and will not be called.\nDrag off pins to build functionality.
)
